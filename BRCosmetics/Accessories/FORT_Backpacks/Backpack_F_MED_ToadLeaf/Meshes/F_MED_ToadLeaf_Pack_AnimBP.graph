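# Anim Blueprint for the ToadLeaf backpack, repositioning its body and swinging two rope chains with Anim Dynamics.
# An additive Body offset feeds gender-selected offsets, weighted by a "Parashuting" alpha derived from spine bend.
# Customization loading sets the gender flag; bIsInFrontEnd picks between root-relative rope setups differing in angular limits.
# A spherical limit collides left rope tip against right; an unused world-space chain remains unconnected.
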AnimBlueprint /BRCosmetics/Accessories/FORT_Backpacks/Backpack_F_MED_ToadLeaf/Meshes/F_MED_ToadLeaf_Pack_AnimBP
target_skeleton=&/F_MED_ToadLeaf_Pack_Skeleton
graphs [
 AnimGraph {nodes [
   n1 = Anim Dynamics - Bone: Dyn_rope_L_01(Alpha=1, ExternalForce="0,0,0") {
    angularDampingOverride=0.999
    angularSpringConstant=500
    bAngularSpring=true
    bOverrideAngularDamping=true
    bUseGravityOverride=true
    simulationSpace=World
    alphaBoolBlend {blendOption=Linear}
    boundBone {boneName=Dyn_rope_L_01}
    componentAppliedLinearAccClamp=<3000 3000 3000>
    componentLinearAccScale=<1 1 1>
    componentLinearVelScale=<1 1 1>
    gravityOverride {z=-980}
    physicsBodyDefinitions [{
      boundBone {boneName=Dyn_rope_L_01}
      boxExtents=<10 10 10>
      constraintSetup {angularLimitsMax=<5 9 0>, angularLimitsMin=<-5 -9 -7>, angularTarget=<1 0 0>}
      localJointOffset=<1 0 0>
     }]
   }
   n2 = Anim Dynamics - Bone: Dyn_rope_L_02(ComponentPose=@n1, Alpha=1, ExternalForce="0,0,0") {
    angularDampingOverride=0.999
    angularSpringConstant=500
    bAngularSpring=true
    bOverrideAngularDamping=true
    bUseGravityOverride=true
    simulationSpace=World
    alphaBoolBlend {blendOption=Linear}
    boundBone {boneName=Dyn_rope_L_02}
    componentAppliedLinearAccClamp=<3000 3000 3000>
    componentLinearAccScale=<1 1 1>
    componentLinearVelScale=<1 1 1>
    gravityOverride {z=-980}
    physicsBodyDefinitions [{
      boundBone {boneName=Dyn_rope_L_02}
      boxExtents=<10 10 10>
      constraintSetup {angularLimitsMax=<5 9 0>, angularLimitsMin=<-5 -9 -7>, angularTarget=<1 0 0>}
      localJointOffset=<1 0 0>
     }]
    relativeSpaceBone {boneName=Dyn_rope_L_01}
   }
   n3 = Anim Dynamics - Bone: Dyn_rope_L_03(ComponentPose=@n2, Alpha=1, ExternalForce="0,0,0") {
    angularDampingOverride=0.999
    angularSpringConstant=500
    bAngularSpring=true
    bOverrideAngularDamping=true
    bUseGravityOverride=true
    bUsePlanarLimit=false
    bUseSphericalLimits=true
    simulationSpace=World
    alphaBoolBlend {blendOption=Linear}
    boundBone {boneName=Dyn_rope_L_03}
    componentAppliedLinearAccClamp=<3000 3000 3000>
    componentLinearAccScale=<1 1 1>
    componentLinearVelScale=<1 1 1>
    gravityOverride {z=-980}
    physicsBodyDefinitions [{
      collisionType=CustomSphere
      sphereCollisionRadius=3
      boundBone {boneName=Dyn_rope_L_03}
      boxExtents=<10 10 10>
      constraintSetup {angularLimitsMax=<5 9 0>, angularLimitsMin=<-5 -9 -7>, angularTarget=<1 0 0>}
      localJointOffset=<1 0 0>
     }]
    relativeSpaceBone {boneName=Dyn_rope_L_02}
    sphericalLimits [{limitRadius=3, drivingBone {boneName=Dyn_rope_R_03}, sphereLocalOffset=<4 0 0>}]
   }
   n4 = Anim Dynamics - Bone: Dyn_rope_R_01(ComponentPose=@n3, Alpha=1, ExternalForce="0,0,0") {
    angularDampingOverride=0.999
    angularSpringConstant=500
    bAngularSpring=true
    bOverrideAngularDamping=true
    bUseGravityOverride=true
    simulationSpace=World
    alphaBoolBlend {blendOption=Linear}
    boundBone {boneName=Dyn_rope_R_01}
    componentAppliedLinearAccClamp=<2500 2500 2500>
    componentLinearAccScale=<0.7 0.7 0.7>
    componentLinearVelScale=<0.7 0.7 0.7>
    gravityOverride {z=-980}
    physicsBodyDefinitions [{
      boundBone {boneName=Dyn_rope_R_01}
      boxExtents=<10 10 10>
      constraintSetup {angularLimitsMax=<5 9 7>, angularLimitsMin=<-5 -9 0>, angularTarget=<1 0 0>}
      localJointOffset=<1 0 0>
     }]
   }
   n5 = Anim Dynamics - Bone: Dyn_rope_R_02(ComponentPose=@n4, Alpha=1, ExternalForce="0,0,0") {
    angularDampingOverride=0.999
    angularSpringConstant=500
    bAngularSpring=true
    bOverrideAngularDamping=true
    bUseGravityOverride=true
    simulationSpace=World
    alphaBoolBlend {blendOption=Linear}
    boundBone {boneName=Dyn_rope_R_02}
    componentAppliedLinearAccClamp=<2500 2500 2500>
    componentLinearAccScale=<0.7 0.7 0.7>
    componentLinearVelScale=<0.7 0.7 0.7>
    gravityOverride {z=-980}
    physicsBodyDefinitions [{
      boundBone {boneName=Dyn_rope_R_02}
      boxExtents=<10 10 10>
      constraintSetup {angularLimitsMax=<5 9 7>, angularLimitsMin=<-5 -9 0>, angularTarget=<1 0 0>}
      localJointOffset=<1 0 0>
     }]
    relativeSpaceBone {boneName=Dyn_rope_R_01}
   }
   Anim Dynamics - Bone: Dyn_rope_R_03(ComponentPose=@n5, Alpha=1, ExternalForce="0,0,0") {
    angularDampingOverride=0.999
    angularSpringConstant=500
    bAngularSpring=true
    bOverrideAngularDamping=true
    bUseGravityOverride=true
    simulationSpace=World
    alphaBoolBlend {blendOption=Linear}
    boundBone {boneName=Dyn_rope_R_03}
    componentAppliedLinearAccClamp=<2500 2500 2500>
    componentLinearAccScale=<0.7 0.7 0.7>
    componentLinearVelScale=<0.7 0.7 0.7>
    gravityOverride {z=-980}
    physicsBodyDefinitions [{
      boundBone {boneName=Dyn_rope_R_03}
      boxExtents=<10 10 10>
      constraintSetup {angularLimitsMax=<5 9 6>, angularLimitsMin=<-5 -9 0>, angularTarget=<1 0 0>}
      localJointOffset=<1 0 0>
     }]
    relativeSpaceBone {boneName=Dyn_rope_R_03}
   }
   n6 = Anim Dynamics - Bone: Dyn_rope_L_01(ComponentPose=Local To Component(LocalPose=Use cached pose 'Base_02'()), Alpha=1, ExternalForce="0,0,0") {
    angularDampingOverride=0.999
    angularSpringConstant=500
    bAngularSpring=true
    bOverrideAngularDamping=true
    bUseGravityOverride=true
    simulationSpace=RootRelative
    alphaBoolBlend {blendOption=Linear}
    boundBone {boneName=Dyn_rope_L_01}
    componentAppliedLinearAccClamp=<3000 3000 3000>
    componentLinearAccScale=<1 1 1>
    componentLinearVelScale=<1 1 1>
    gravityOverride {z=-980}
    physicsBodyDefinitions [{
      boundBone {boneName=Dyn_rope_L_01}
      boxExtents=<10 10 10>
      constraintSetup {angularLimitsMax=<5 7 0>, angularLimitsMin=<-5 -7 -7>, angularTarget=<1 0 0>}
      localJointOffset=<1 0 0>
     }]
   }
   n7 = Anim Dynamics - Bone: Dyn_rope_L_02(ComponentPose=@n6, Alpha=1, ExternalForce="0,0,0") {
    angularDampingOverride=0.999
    angularSpringConstant=500
    bAngularSpring=true
    bOverrideAngularDamping=true
    bUseGravityOverride=true
    simulationSpace=RootRelative
    alphaBoolBlend {blendOption=Linear}
    boundBone {boneName=Dyn_rope_L_02}
    componentAppliedLinearAccClamp=<3000 3000 3000>
    componentLinearAccScale=<1 1 1>
    componentLinearVelScale=<1 1 1>
    gravityOverride {z=-980}
    physicsBodyDefinitions [{
      boundBone {boneName=Dyn_rope_L_02}
      boxExtents=<10 10 10>
      constraintSetup {angularLimitsMax=<5 5 0>, angularLimitsMin=<-5 -5 -7>, angularTarget=<1 0 0>}
      localJointOffset=<1 0 0>
     }]
   }
   n8 = Anim Dynamics - Bone: Dyn_rope_L_03(ComponentPose=@n7, Alpha=1, ExternalForce="0,0,0") {
    angularDampingOverride=0.999
    angularSpringConstant=500
    bAngularSpring=true
    bOverrideAngularDamping=true
    bUseGravityOverride=true
    bUsePlanarLimit=false
    bUseSphericalLimits=true
    simulationSpace=RootRelative
    alphaBoolBlend {blendOption=Linear}
    boundBone {boneName=Dyn_rope_L_03}
    componentAppliedLinearAccClamp=<3000 3000 3000>
    componentLinearAccScale=<1 1 1>
    componentLinearVelScale=<1 1 1>
    gravityOverride {z=-980}
    physicsBodyDefinitions [{
      collisionType=CustomSphere
      sphereCollisionRadius=2.5
      boundBone {boneName=Dyn_rope_L_03}
      boxExtents=<10 10 10>
      constraintSetup {angularLimitsMax=<5 3 0>, angularLimitsMin=<-5 -3 -7>, angularTarget=<1 0 0>}
      localJointOffset=<1 0 0>
     }]
    sphericalLimits [{limitRadius=3, drivingBone {boneName=Dyn_rope_R_03}, sphereLocalOffset=<4 0 0>}]
   }
   n9 = Anim Dynamics - Bone: Dyn_rope_R_01(ComponentPose=@n8, Alpha=1, ExternalForce="0,0,0") {
    angularDampingOverride=0.999
    angularSpringConstant=500
    bAngularSpring=true
    bOverrideAngularDamping=true
    bUseGravityOverride=true
    simulationSpace=RootRelative
    alphaBoolBlend {blendOption=Linear}
    boundBone {boneName=Dyn_rope_R_01}
    componentAppliedLinearAccClamp=<2500 2500 2500>
    componentLinearAccScale=<0.7 0.7 0.7>
    componentLinearVelScale=<0.7 0.7 0.7>
    gravityOverride {z=-980}
    physicsBodyDefinitions [{
      boundBone {boneName=Dyn_rope_R_01}
      boxExtents=<10 10 10>
      constraintSetup {angularLimitsMax=<5 7 7>, angularLimitsMin=<-5 -7 0>, angularTarget=<1 0 0>}
      localJointOffset=<0.8 0 0>
     }]
   }
   n10 = Anim Dynamics - Bone: Dyn_rope_R_02(ComponentPose=@n9, Alpha=1, ExternalForce="0,0,0") {
    angularDampingOverride=0.999
    angularSpringConstant=500
    bAngularSpring=true
    bOverrideAngularDamping=true
    bUseGravityOverride=true
    simulationSpace=RootRelative
    alphaBoolBlend {blendOption=Linear}
    boundBone {boneName=Dyn_rope_R_02}
    componentAppliedLinearAccClamp=<2500 2500 2500>
    componentLinearAccScale=<0.7 0.7 0.7>
    componentLinearVelScale=<0.7 0.7 0.7>
    gravityOverride {z=-980}
    physicsBodyDefinitions [{
      boundBone {boneName=Dyn_rope_R_02}
      boxExtents=<10 10 10>
      constraintSetup {angularLimitsMax=<5 5 7>, angularLimitsMin=<-5 -5 0>, angularTarget=<1 0 0>}
      localJointOffset=<0.8 0 0>
     }]
   }
   n11 = Anim Dynamics - Bone: Dyn_rope_R_03(ComponentPose=@n10, Alpha=1, ExternalForce="0,0,0") {
    angularDampingOverride=0.999
    angularSpringConstant=500
    bAngularSpring=true
    bOverrideAngularDamping=true
    bUseGravityOverride=true
    simulationSpace=RootRelative
    alphaBoolBlend {blendOption=Linear}
    boundBone {boneName=Dyn_rope_R_03}
    componentAppliedLinearAccClamp=<2500 2500 2500>
    componentLinearAccScale=<0.7 0.7 0.7>
    componentLinearVelScale=<0.7 0.7 0.7>
    gravityOverride {z=-980}
    physicsBodyDefinitions [{
      boundBone {boneName=Dyn_rope_R_03}
      boxExtents=<10 10 10>
      constraintSetup {angularLimitsMax=<5 3 6>, angularLimitsMin=<-5 -3 0>, angularTarget=<1 0 0>}
      localJointOffset=<0.8 0 0>
     }]
   }
   n12 = Anim Dynamics - Bone: Dyn_rope_L_01(ComponentPose=Local To Component(LocalPose=Use cached pose 'Base_02'()), Alpha=1, ExternalForce="0,0,0") {
    angularDampingOverride=0.999
    angularSpringConstant=500
    bAngularSpring=true
    bOverrideAngularDamping=true
    bUseGravityOverride=true
    simulationSpace=RootRelative
    alphaBoolBlend {blendOption=Linear}
    boundBone {boneName=Dyn_rope_L_01}
    componentAppliedLinearAccClamp=<3000 3000 3000>
    componentLinearAccScale=<1 1 1>
    componentLinearVelScale=<1 1 1>
    gravityOverride {z=-980}
    physicsBodyDefinitions [{
      boundBone {boneName=Dyn_rope_L_01}
      boxExtents=<10 10 10>
      constraintSetup {angularLimitsMax=<5 1 0>, angularLimitsMin=<-5 -7 -7>, angularTarget=<1 0 0>}
      localJointOffset=<1 0 0>
     }]
   }
   n13 = Anim Dynamics - Bone: Dyn_rope_L_02(ComponentPose=@n12, Alpha=1, ExternalForce="0,0,0") {
    angularDampingOverride=0.999
    angularSpringConstant=500
    bAngularSpring=true
    bOverrideAngularDamping=true
    bUseGravityOverride=true
    simulationSpace=RootRelative
    alphaBoolBlend {blendOption=Linear}
    boundBone {boneName=Dyn_rope_L_02}
    componentAppliedLinearAccClamp=<3000 3000 3000>
    componentLinearAccScale=<1 1 1>
    componentLinearVelScale=<1 1 1>
    gravityOverride {z=-980}
    physicsBodyDefinitions [{
      boundBone {boneName=Dyn_rope_L_02}
      boxExtents=<10 10 10>
      constraintSetup {angularLimitsMax=<5 1 0>, angularLimitsMin=<-5 -5 -7>, angularTarget=<1 0 0>}
      localJointOffset=<1 0 0>
     }]
   }
   n14 = Anim Dynamics - Bone: Dyn_rope_L_03(ComponentPose=@n13, Alpha=1, ExternalForce="0,0,0") {
    angularDampingOverride=0.999
    angularSpringConstant=500
    bAngularSpring=true
    bOverrideAngularDamping=true
    bUseGravityOverride=true
    bUsePlanarLimit=false
    bUseSphericalLimits=true
    simulationSpace=RootRelative
    alphaBoolBlend {blendOption=Linear}
    boundBone {boneName=Dyn_rope_L_03}
    componentAppliedLinearAccClamp=<3000 3000 3000>
    componentLinearAccScale=<1 1 1>
    componentLinearVelScale=<1 1 1>
    gravityOverride {z=-980}
    physicsBodyDefinitions [{
      collisionType=CustomSphere
      sphereCollisionRadius=2.5
      boundBone {boneName=Dyn_rope_L_03}
      boxExtents=<10 10 10>
      constraintSetup {angularLimitsMax=<5 3 0>, angularLimitsMin=<-5 -3 -7>, angularTarget=<1 0 0>}
      localJointOffset=<1 0 0>
     }]
    sphericalLimits [{limitRadius=3, drivingBone {boneName=Dyn_rope_R_03}, sphereLocalOffset=<4 0 0>}]
   }
   n15 = Anim Dynamics - Bone: Dyn_rope_R_01(ComponentPose=@n14, Alpha=1, ExternalForce="0,0,0") {
    angularDampingOverride=0.999
    angularSpringConstant=500
    bAngularSpring=true
    bOverrideAngularDamping=true
    bUseGravityOverride=true
    simulationSpace=RootRelative
    alphaBoolBlend {blendOption=Linear}
    boundBone {boneName=Dyn_rope_R_01}
    componentAppliedLinearAccClamp=<2500 2500 2500>
    componentLinearAccScale=<0.7 0.7 0.7>
    componentLinearVelScale=<0.7 0.7 0.7>
    gravityOverride {z=-980}
    physicsBodyDefinitions [{
      boundBone {boneName=Dyn_rope_R_01}
      boxExtents=<10 10 10>
      constraintSetup {angularLimitsMax=<5 0 7>, angularLimitsMin=<-5 -7 0>, angularTarget=<1 0 0>}
      localJointOffset=<0.8 0 0>
     }]
   }
   n16 = Anim Dynamics - Bone: Dyn_rope_R_02(ComponentPose=@n15, Alpha=1, ExternalForce="0,0,0") {
    angularDampingOverride=0.999
    angularSpringConstant=500
    bAngularSpring=true
    bOverrideAngularDamping=true
    bUseGravityOverride=true
    simulationSpace=RootRelative
    alphaBoolBlend {blendOption=Linear}
    boundBone {boneName=Dyn_rope_R_02}
    componentAppliedLinearAccClamp=<2500 2500 2500>
    componentLinearAccScale=<0.7 0.7 0.7>
    componentLinearVelScale=<0.7 0.7 0.7>
    gravityOverride {z=-980}
    physicsBodyDefinitions [{
      boundBone {boneName=Dyn_rope_R_02}
      boxExtents=<10 10 10>
      constraintSetup {angularLimitsMax=<5 1 7>, angularLimitsMin=<-5 -5 0>, angularTarget=<1 0 0>}
      localJointOffset=<0.8 0 0>
     }]
   }
   n17 = Anim Dynamics - Bone: Dyn_rope_R_03(ComponentPose=@n16, Alpha=1, ExternalForce="0,0,0") {
    angularDampingOverride=0.999
    angularSpringConstant=500
    bAngularSpring=true
    bOverrideAngularDamping=true
    bUseGravityOverride=true
    simulationSpace=RootRelative
    alphaBoolBlend {blendOption=Linear}
    boundBone {boneName=Dyn_rope_R_03}
    componentAppliedLinearAccClamp=<2500 2500 2500>
    componentLinearAccScale=<0.7 0.7 0.7>
    componentLinearVelScale=<0.7 0.7 0.7>
    gravityOverride {z=-980}
    physicsBodyDefinitions [{
      boundBone {boneName=Dyn_rope_R_03}
      boxExtents=<10 10 10>
      constraintSetup {angularLimitsMax=<5 3 6>, angularLimitsMin=<-5 -3 0>, angularTarget=<1 0 0>}
      localJointOffset=<0.8 0 0>
     }]
   }
   Output Pose(Result=Blend Poses by bool(BlendPose_0=Component To Local(ComponentPose=@n17), BlendPose_1=Component To Local(ComponentPose=@n11), bActiveValue=Get bIsInFrontEnd()) {blendType=Linear}) {layerGroup=None}
   n18 = "Transform (Modify) Bone - Bone: Body"(
    ComponentPose=Local To Component(LocalPose=Use cached pose 'Base_01'())
    Alpha=Get Parashuting() -> (Parashuting:double)
    Translation="0,-8,5"
    Rotation="0,0,35"
    Scale="1,1,1"
   ) {
    rotationMode=BMM_Additive
    scaleMode=BMM_Additive
    translationMode=BMM_Additive
    alphaBoolBlend {blendOption=Linear}
    boneToModify {boneName=Body}
    rotation {roll=35}
    translation {y=-8, z=5}
   }
   n19 = "Transform (Modify) Bone - Bone: Body"(
    ComponentPose=Local To Component(LocalPose=Use cached pose 'Base_01'())
    Alpha=Get Parashuting()
    Translation="0,-12,6"
    Rotation="0,0,50"
    Scale="1,1,1"
   ) {
    rotationMode=BMM_Additive
    scaleMode=BMM_Additive
    translationMode=BMM_Additive
    alphaBoolBlend {blendOption=Linear}
    boneToModify {boneName=Body}
    rotation {roll=50}
    translation {y=-12, z=6}
   }
   Save cached pose 'Base_02'(Pose=Blend Poses by bool(
     BlendPose_0=Component To Local(ComponentPose=@n19)
     BlendPose_1=Component To Local(ComponentPose=@n18)
     bActiveValue=Get isFemale() -> (isFemale:bool)
    ) {blendType=Linear}) {cacheName=Base_02}
   n20 = "Transform (Modify) Bone - Bone: Body"(ComponentPose=Local To Component(LocalPose=Input Pose()), Alpha=1, Translation="0,1,0", Rotation="0,0,10", Scale="1,1,1") {
    rotationMode=BMM_Additive
    scaleMode=BMM_Additive
    translationMode=BMM_Additive
    alphaBoolBlend {blendOption=Linear}
    boneToModify {boneName=Body}
    rotation {roll=10}
    translation {y=1}
   }
   Save cached pose 'Base_01'(Pose=Component To Local(ComponentPose=@n20)) {cacheName=Base_01}
  ]}
 EventGraph {nodes [
   Event BlueprintUpdateAnimation()
   n1 = GetDeltaTransformFromRefPose!(self=GetOwnerSkeletalMeshComponent(), BoneName=spine_05, BaseName=spine_01) -> (ReturnValue_Location:Vector, ReturnValue_Rotation:Rotator, ReturnValue_Scale:Vector)
   > Set Parashuting(Parashuting:double=MapRangeClamped(Value=BreakRotator(InRot=@n1.ReturnValue_Rotation).Yaw, InRangeA=-25, InRangeB=-65)) -> (Output_Get:double)
   n2 = Event OnCharacterCustomizationLoaded()
   > n3 = Switch on EFortCustomGender!(
    execute
    Selection:EFortCustomGender Enum=Get GenderPermitted!(self:Custom Character Part Object Reference="Get (a copy)"(Array:Array of Custom Character Part Object References=@n2.CustomParts, Dimension 1=1) -> (Output:Custom Character Part Object Reference)) -> (GenderPermitted:EFortCustomGender Enum=Invalid)
   ) -> (Invalid:Exec, Male:Exec, Female:Exec, Both:Exec) {is=title K2Node_SwitchEnum}
   Set isFemale(execute=@n3.Female, isFemale:bool=true) -> (Output_Get:bool)
   TryGetPawnOwner()
  ]}
]
variables [
 isFemale {type=boolean}
 Parashuting {type=number}
]
